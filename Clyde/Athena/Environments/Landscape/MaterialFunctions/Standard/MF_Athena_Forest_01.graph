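# Material function for the forest floor layer of the Athena landscape, built on one overridable "Forest" texture parameter.
# The texture samples with the UVs input, defaulting to standard texture coordinates, through a shared world-group wrap sampler.
# RGB supplies BaseColor and alpha supplies Height, both exposed as function outputs.
# A material attributes output also packs them, with height in the Opacity slot and a constant, fully rough Roughness.
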MaterialFunction /Clyde/Athena/Environments/Landscape/MaterialFunctions/Standard/MF_Athena_Forest_01
graph {nodes [
  n1 = TextureSampleParameter2D(UVs=FunctionInput(Preview=TextureCoordinate()) {bUsePreviewValueAsDefault=true, inputName=UVs, inputType=FunctionInput_Vector2}) {
   group=Forest
   parameterName=Forest
   samplerSource=SSM_Wrap_WorldGroupSettings
   texture=/Game/Packages/DS_Fortnight/SM/Textures/T_FORT_ForestFloor01
  }
  FunctionOutput(None=MakeMaterialAttributes!(
    BaseColor=FunctionOutput(None=@n1.RGB) {outputName=BaseColor}
    Metallic
    Specular
    Roughness=Constant() {r=1}
    Anisotropy
    EmissiveColor
    Opacity=FunctionOutput(None=@n1.A) {outputName=Height, sortPriority=1}
    OpacityMask
    Normal
    Tangent
    WorldPositionOffset
    SubsurfaceColor
    ClearCoat
    ClearCoatRoughness
    AmbientOcclusion
    Refraction
    CustomizedUVs_0
    CustomizedUVs_1
    CustomizedUVs_2
    CustomizedUVs_3
    CustomizedUVs_4
    CustomizedUVs_5
    CustomizedUVs_6
    CustomizedUVs_7
    PixelDepthOffset
    ShadingModel
    Displacement
   ))
 ]}
