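# Landscape material function for Athena that supplies road, gravel and mud colors plus height outputs from one packed grass biome mask texture.
# The texture is sampled with world XY projection, and its channels feed road shade, gravel height, gravel shade and road height.
# Gravel and road each blend two tints using CheapContrast of their own channel, while mud remaps the gravel channel into a brown tint.
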MaterialFunction /Clyde/Athena/Environments/Landscape/MaterialFunctions/Standard/MF_Athena_RoadGravel
graph {nodes [
  n1 = WorldPosition()
  n2 = TextureSample(UVs=Divide(A=ComponentMask(None=@n1.XYZ) {g=true, r=true}, B=Constant() {r=700})) {
   samplerSource=SSM_Wrap_WorldGroupSettings
   samplerType=SAMPLERTYPE_LinearColor
   texture=/Game/Athena/Environments/Landscape/Textures/Athena_GrassBiome_M_02
  }
  n3 = FunctionOutput(None=@n2.B) {outputName=Gravel_NoColor}
  n4 = Constant3Vector() {constant=<rgba 0.1979 0.1709 0.1554 0>}
  n5 = Constant3Vector() {constant=<rgba 0.244 0.2799 0.2917 0>}
  n6 = Constant() {r=2}
  FunctionOutput(None=LinearInterpolate(
    A=Multiply(A=@n3, B=@n4)
    B=Multiply(A=@n3, B=@n5)
    Alpha=MaterialFunctionCall(In=@n3, Contrast=@n6) {materialFunction=&CheapContrast}.Result
   )) {outputName=Gravel}
  FunctionOutput(None=@n2.A) {outputName=Road_Height}
  FunctionOutput(None=@n2.G) {outputName=Gravel_Height}
  n7 = Multiply(A=@n2.R, B=Constant3Vector() {constant=<rgba 0.4 0.3642 0.3854 0>})
  n8 = Multiply(A=@n2.R, B=Constant3Vector() {constant=<rgba 0.3229 0.3229 0.3229 0>})
  n9 = MaterialFunctionCall(In=@n2.R, Contrast=Constant() {r=0.5}) {materialFunction=&CheapContrast}
  FunctionOutput(None=Clamp(None=LinearInterpolate(A=@n7, B=@n8, Alpha=@n9.Result)) {minDefault=0.05}) {outputName=Road}
  n10 = Constant() {r=-0.25}
  n11 = Constant() {r=1}
  FunctionOutput(None=Multiply(A=LinearInterpolate(A=@n10, B=@n11, Alpha=@n2.B), B=Constant3Vector() {constant=<rgba 0.1198 0.0917 0.0705 0>})) {outputName=Mud}
 ]}
