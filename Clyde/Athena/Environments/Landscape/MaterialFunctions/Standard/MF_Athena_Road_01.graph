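# Landscape material function that supplies a street road surface layer for the Athena environment.
# It samples one road diffuse texture with a shared wrap sampler, using a UV input that defaults to mesh texture coordinates.
# Slightly brightened RGB becomes BaseColor and alpha becomes Height, each exposed as an output.
# Both also feed a Material Attributes output, with height packed into Opacity, fully rough shading and no normal map.
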MaterialFunction /Clyde/Athena/Environments/Landscape/MaterialFunctions/Standard/MF_Athena_Road_01
graph {nodes [
  n1 = TextureSample(UVs=FunctionInput(Preview=TextureCoordinate()) {bUsePreviewValueAsDefault=true, inputType=FunctionInput_Vector2}) {samplerSource=SSM_Wrap_WorldGroupSettings, texture=/Game/Packages/DS_Fortnight/SM/Textures/T_Street_Road_D}
  n2 = Multiply(A=@n1.RGB, B=Constant() {r=1.1})
  FunctionOutput(None=MakeMaterialAttributes!(
    BaseColor=FunctionOutput(None=@n2) {outputName=BaseColor}
    Metallic
    Specular
    Roughness=Constant() {r=1}
    Anisotropy
    EmissiveColor
    Opacity=FunctionOutput(None=@n1.A) {outputName=Height}
    OpacityMask
    Normal
    Tangent
    WorldPositionOffset
    SubsurfaceColor
    ClearCoat
    ClearCoatRoughness
    AmbientOcclusion
    Refraction
    CustomizedUVs_0
    CustomizedUVs_1
    CustomizedUVs_2
    CustomizedUVs_3
    CustomizedUVs_4
    CustomizedUVs_5
    CustomizedUVs_6
    CustomizedUVs_7
    PixelDepthOffset
    ShadingModel
    Displacement
   ))
 ]}
